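# Control Rig for the M_LRG_ServeStreet body that retargets the base skeleton pose onto its differently proportioned deform_ bones.
# Arms use constraints, full-body IK and a twist correction weighted by distance to the head.
# Legs use aim and parent constraints, with foot-height-driven calf rotation when shoeless; fingers map onto thumb, index and middle chains.
# Joint bend drives mirrored elbow and knee correctives, and spine_05 twist is copied down to the pelvis.
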ControlRigBlueprint /BRCosmetics/Characters/Player/Male/Large/Bodies/M_LRG_ServeStreet/Meshes/M_LRG_ServeStreet_CtrlRig
graphs [
 ContainedGraph {nodes [
   n1 = RigVMFunctionEntryNode()
   n2 = &RigUnit_GetTransform(Space=LocalSpace, Item=@n1.Item)
   n3 = &RigUnit_GetTransform(Space=LocalSpace, bInitial=true, Item=@n1.Item)
   > RigVMFunctionReturnNode(OutTransform=/Script/RigVM.RigVMFunction_MathTransformToEulerTransform(Value=&RigVMFunction_MathTransformMakeRelative(Global=@n2.Transform, Parent=@n3.Transform).Local) {is=struct RigVMUnitNode, template="Cast::Execute(in Value,out Result)"}.Result)
  ]}
 New Function_ContainedGraph {nodes [
   RigVMFunctionEntryNode()
   > RigVMFunction_Sequence = RigVMAggregateNode()
   &RigUnit_AimConstraintLocalSpaceOffset(
    Child=(Type=Bone,Name="deform_upperarm_l")
    Filter=(bX=True,bY=True,bZ=True)
    Parents=((Item=(Type=Bone,Name="lowerarm_l"),Weight=1))
    Weight=1
    WorldUp=(Target=(X=0,Y=0,Z=1),Kind=Location,Space=(Type=Bone,Name="lowerarm_l"))
    ExecutePin=@RigVMFunction_Sequence.B
   )
   > &RigUnit_ParentConstraint(
    AdvancedSettings=(InterpolationType=Average,RotationOrderForFilter=XZY)
    Child=(Type=Bone,Name="deform_lowerarm_l")
    Parents=((Item=(Type=Bone,Name="lowerarm_l"),Weight=1))
    Weight=1
   )
   > &RigUnit_ParentConstraint(
    AdvancedSettings=(InterpolationType=Average,RotationOrderForFilter=XZY)
    Child=(Type=Bone,Name="deform_hand_l")
    Parents=((Item=(Type=Bone,Name="hand_l"),Weight=1))
    Weight=1
   )
   > &RigUnit_AimConstraintLocalSpaceOffset(
    Child=(Type=Bone,Name="deform_upperarm_r")
    Filter=(bX=True,bY=True,bZ=True)
    Parents=((Item=(Type=Bone,Name="cr_lowerarm_push_r"),Weight=1))
    Weight=1
    WorldUp=(Target=(X=0,Y=0,Z=1),Kind=Location,Space=(Type=Bone,Name="cr_lowerarm_push_r"))
   )
   > &RigUnit_ParentConstraint(
    AdvancedSettings=(InterpolationType=Average,RotationOrderForFilter=XZY)
    Child=(Type=Bone,Name="deform_lowerarm_r")
    Parents=((Item=(Type=Bone,Name="cr_lowerarm_push_r"),Weight=1))
    Weight=1
   )
   > &RigUnit_ParentConstraint(
    AdvancedSettings=(InterpolationType=Average,RotationOrderForFilter=XZY)
    Child=(Type=Bone,Name="deform_hand_r")
    Parents=((Item=(Type=Bone,Name="hand_r"),Weight=1))
    Weight=1
   )
   > /Script/ControlRig.RigUnit_PositionConstraintLocalSpaceOffset(
    Child=(Type=Bone,Name="deform_upperarm_l_end")
    Filter=(bX=True,bY=True,bZ=True)
    Parents=((Item=(Type=Bone,Name="deform_lowerarm_l"),Weight=1))
    Weight=1
    bMaintainOffset=True
   ) {is=struct RigVMUnitNode}
   > &RigUnit_PositionConstraintLocalSpaceOffset(
    Child=(Type=Bone,Name="deform_upperarm_r_end")
    Filter=(bX=True,bY=True,bZ=True)
    Parents=((Item=(Type=Bone,Name="deform_lowerarm_r"),Weight=1))
    Weight=1
    bMaintainOffset=True
   )
   /Script/PBIK.RigUnit_PBIK(
    BoneSettings=((Bone="deform_hand_l",RotationStiffness=0,PositionStiffness=0,X=Free,MinX=0,MaxX=0,Y=Free,MinY=0,MaxY=0,Z=Free,MinZ=0,MaxZ=0,bUsePreferredAngles=true,PreferredAngles=(X=1,Y=0,Z=0)))
    Debug=(DrawScale=1,bDrawDebug=false)
    Effectors=((Bone="hand_l",Transform=(Rotation=(X=0,Y=0,Z=0,W=1),Translation=(X=0,Y=0,Z=0),Scale3D=(X=1,Y=1,Z=1)),PositionAlpha=1,RotationAlpha=1,StrengthAlpha=1,PullChainAlpha=1,PinRotation=1,ChainDepth=0))
    ExcludedBones=()
    Root=deform_upperarm_l
    Settings="(Iterations=20,SubIterations=0,MassMultiplier=1,bAllowStretch=false,RootBehavior=PrePull,PrePullRootSettings=(RotationAlpha=0,RotationAlphaX=1,RotationAlphaY=1,RotationAlphaZ=1,PositionAlpha=1,PositionAlphaX=1,PositionAlphaY=1,PositionAlphaZ=1),GlobalPullC ...(truncated)"
    ExecutePin=@RigVMFunction_Sequence.C
   ) {is=struct RigVMUnitNode}
   > &RigUnit_PBIK(
    BoneSettings="((Bone=\"deform_hand_r\",RotationStiffness=0,PositionStiffness=0,X=Free,MinX=0,MaxX=0,Y=Free,MinY=0,MaxY=0,Z=Free,MinZ=0,MaxZ=0,bUsePreferredAngles=false,PreferredAngles=(X=1,Y=0,Z=0)),(Bone=\"deform_upperarm_l\",RotationStiffness=0,PositionStiffness=0,X=Free, ...(truncated)"
    Debug=(DrawScale=3,bDrawDebug=true)
    Effectors="((Bone=\"hand_r\",Transform=(Rotation=(X=0,Y=0,Z=0,W=1),Translation=(X=0,Y=0,Z=0),Scale3D=(X=1,Y=1,Z=1)),PositionAlpha=1,RotationAlpha=1,StrengthAlpha=1,PullChainAlpha=1,PinRotation=1,ChainDepth=0),(Bone=\"upperarm_r\",Transform=(Rotation=(X=0,Y=0,Z=0,W=1),Tra ...(truncated)"
    ExcludedBones=()
    Root=deform_upperarm_r
    Settings="(Iterations=20,SubIterations=0,MassMultiplier=1,bAllowStretch=true,RootBehavior=PrePull,PrePullRootSettings=(RotationAlpha=0,RotationAlphaX=1,RotationAlphaY=1,RotationAlphaZ=1,PositionAlpha=1,PositionAlphaX=1,PositionAlphaY=1,PositionAlphaZ=1),GlobalPullCh ...(truncated)"
   )
   > ApplyFBIKChainCorrection_1_1 = RigVMFunctionReferenceNode(
    End=(Type=Bone,Name="deform_hand_r")
    QuatAmount=2.5
    Start=(Type=Bone,Name="deform_upperarm_r")
    Weight=RigVMVariableNode() {variable=R Elbow Distance to Head}.Value
   )
   > ApplyFBIKChainCorrection_1_2 = RigVMFunctionReferenceNode(
    End=(Type=Bone,Name="deform_hand_l")
    QuatAmount=2.5
    Start=(Type=Bone,Name="deform_upperarm_l")
    Weight=RigVMVariableNode() {variable=L Elbow Distance to Head}.Value
   )
   RigVMFunctionReturnNode(ExecuteContext=@RigVMFunction_Sequence.D)
   Get Normalized Distance Between_1_1 = RigVMFunctionReferenceNode(
    Debug=true
    Max Distance=35
    Min Distance=1
    Source=(Type=Bone,Name="lowerarm_r")
    Target=(Type=Bone,Name="head")
    ExecuteContext=@RigVMFunction_Sequence.A
   )
   > RigVMVariableNode(Value=@"Get Normalized Distance Between_1_1".Result) {variable=R Elbow Distance to Head}
   > &RigUnit_ModifyTransforms(
    ItemToModify=((Item=(Type=Bone,Name="cr_lowerarm_push_r"),Transform=(Rotation=(X=0,Y=0,Z=0,W=1),Translation=(X=0,Y=5,Z=0),Scale3D=(X=1,Y=1,Z=1))))
    WeightMinimum=0
    Weight=RigVMVariableNode() {variable=R Elbow Distance to Head}.Value
   )
   > &RigUnit_AimConstraintLocalSpaceOffset(
    Child=(Type=Bone,Name="cr_lowerarm_push_r")
    Filter=(bX=True,bY=True,bZ=True)
    Parents=((Item=(Type=Bone,Name="hand_r"),Weight=1))
    Weight=1
    WorldUp=(Target=(X=0,Y=0,Z=1),Kind=Location,Space=(Type=Bone,Name="hand_r"))
   )
   > Get Normalized Distance Between_1_2 = RigVMFunctionReferenceNode(Debug=true, Max Distance=35, Min Distance=1, Source=(Type=Bone,Name="hand_l"), Target=(Type=Bone,Name="head"))
   > RigVMVariableNode(Value=@"Get Normalized Distance Between_1_2".Result) {variable=L Elbow Distance to Head}
   > &RigUnit_ModifyTransforms(
    ItemToModify=((Item=(Type=Bone,Name="cr_lowerarm_push_l"),Transform=(Rotation=(X=0,Y=0,Z=0,W=1),Translation=(X=0,Y=5,Z=0),Scale3D=(X=1,Y=1,Z=1))))
    WeightMinimum=0
    Weight=RigVMVariableNode() {variable=L Elbow Distance to Head}.Value
   )
   > &RigUnit_AimConstraintLocalSpaceOffset(
    Child=(Type=Bone,Name="cr_lowerarm_push_l")
    Filter=(bX=True,bY=True,bZ=True)
    Parents=((Item=(Type=Bone,Name="hand_l"),Weight=1))
    Weight=1
    WorldUp=(Target=(X=0,Y=0,Z=1),Kind=Location,Space=(Type=Bone,Name="hand_l"))
   )
  ]}
 New Function_ContainedGraph {nodes [
   n1 = RigVMFunctionEntryNode()
   RigVMFunction_Sequence = &RigVMFunction_Sequence(ExecuteContext=@n1.ExecuteContext)
   &RigUnit_AimConstraintLocalSpaceOffset(
    Child=(Type=Bone,Name="deform_thigh_l")
    Filter=(bX=True,bY=True,bZ=True)
    Parents=((Item=(Type=Bone,Name="calf_l"),Weight=1))
    Weight=1
    WorldUp=(Target=(X=0,Y=0,Z=1),Kind=Location,Space=(Type=Bone,Name="calf_l"))
    ExecutePin=@RigVMFunction_Sequence.A
   )
   > &RigUnit_AimConstraintLocalSpaceOffset(
    Child=(Type=Bone,Name="deform_thigh_r")
    Filter=(bX=True,bY=True,bZ=True)
    Parents=((Item=(Type=Bone,Name="calf_r"),Weight=1))
    Weight=1
    WorldUp=(Target=(X=0,Y=0,Z=1),Kind=Location,Space=(Type=Bone,Name="calf_r"))
   )
   > &RigUnit_ParentConstraint(
    AdvancedSettings=(InterpolationType=Average,RotationOrderForFilter=XZY)
    Child=(Type=Bone,Name="deform_calf_l")
    Parents=((Item=(Type=Bone,Name="calf_l"),Weight=1))
    Weight=1
   )
   > n2 = &RigUnit_ParentConstraint(
    AdvancedSettings=(InterpolationType=Average,RotationOrderForFilter=XZY)
    Child=(Type=Bone,Name="deform_calf_r")
    Parents=((Item=(Type=Bone,Name="calf_r"),Weight=1))
    Weight=1
   )
   RigVMFunction_ControlFlowBranch = &RigVMFunction_ControlFlowBranch(ExecuteContext=@n2.ExecutePin, Condition=RigVMVariableNode() {variable=Wearing Shoes}.Value)
   RigUnit_GetRelativeTransformForItem = &RigUnit_GetRelativeTransformForItem(Child=(Type=Bone,Name="ball_r"), Parent=(Type=Bone,Name="root"), bChildInitial=False, bParentInitial=false)
   RigVMFunction_MathVectorDistance = &RigVMFunction_MathVectorDistance(
    A=(X=0,Y=0,Z=0)
    B=(X=0,Y=0,Z=0)
    B.Z=@RigUnit_GetRelativeTransformForItem.RelativeTransform.Translation.Z
    A.Z=&RigUnit_GetRelativeTransformForItem(Child=(Type=Bone,Name="foot_r"), Parent=(Type=Bone,Name="root"), bChildInitial=False, bParentInitial=false).RelativeTransform.Translation.Z
   )
   Remap = &RigVMFunction_MathDoubleRemap(SourceMaximum=12, SourceMinimum=6, TargetMaximum=1, TargetMinimum=0, Value=@RigVMFunction_MathVectorDistance.Result)
   &RigUnit_ModifyTransforms(
    ItemToModify=((Item=(Type=Bone,Name="deform_calf_r"),Transform=(Rotation=(X=0,Y=0,Z=0.1525,W=-0.9883),Translation=(X=0,Y=0,Z=0),Scale3D=(X=1,Y=1,Z=1))))
    WeightMinimum=0
    Weight=@Remap.Result
    ExecutePin=@RigVMFunction_ControlFlowBranch.False
   )
   n3 = &RigUnit_GetRelativeTransformForItem(Child=(Type=Bone,Name="foot_l"), Parent=(Type=Bone,Name="root"), bChildInitial=False, bParentInitial=false)
   > &RigUnit_ModifyTransforms(
    ItemToModify=((Item=(Type=Bone,Name="deform_calf_l"),Transform=(Rotation=(X=0,Y=0,Z=0.1525,W=-0.9883),Translation=(X=0,Y=0,Z=0),Scale3D=(X=1,Y=1,Z=1))))
    WeightMinimum=0
    Weight=&RigVMFunction_MathDoubleRemap(
     SourceMaximum=12
     SourceMinimum=6
     TargetMaximum=1
     TargetMinimum=0
     Value=&RigVMFunction_MathVectorDistance(
      A=(X=0,Y=0,Z=0)
      B=(X=0,Y=0,Z=0)
      A.Z=@RigUnit_GetRelativeTransformForItem.RelativeTransform.Translation.Z
      B.Z=@n3.RelativeTransform.Translation.Z
     ).Result
    ).Result
   )
   RigVMFunctionReturnNode(ExecuteContext=@RigVMFunction_Sequence.B)
  ]}
 New Function_ContainedGraph {nodes [
   RigVMFunctionEntryNode()
   > &RigUnit_ParentConstraint(
    AdvancedSettings=(InterpolationType=Average,RotationOrderForFilter=XZY)
    Child=(Type=Bone,Name="deform_elbow_in_l_base")
    Parents=((Item=(Type=Bone,Name="deform_lowerarm_l"),Weight=1),(Item=(Type=Bone,Name="deform_upperarm_l_end"),Weight=1))
    Weight=1
   )
   > &RigUnit_ParentConstraint(
    AdvancedSettings=(InterpolationType=Average,RotationOrderForFilter=XZY)
    Child=(Type=Bone,Name="deform_elbow_in_r_base")
    Parents=((Item=(Type=Bone,Name="deform_lowerarm_r"),Weight=1),(Item=(Type=Bone,Name="deform_upperarm_r_end"),Weight=1))
    Weight=1
   )
   > Get Delta Ref = RigVMFunctionReferenceNode(Item=(Type=Bone,Name="deform_lowerarm_l"))
   > Get Delta Ref_1 = RigVMFunctionReferenceNode(Item=(Type=Bone,Name="deform_lowerarm_r"))
   > Modify Transforms with Mirror = RigVMFunctionReferenceNode(
    Base Modify Data=((Item=(Type=Bone,Name="deform_elbow_in_l"),Transform=(Rotation=(X=0,Y=0,Z=-0.1305,W=0.9914),Translation=(X=1,Y=4,Z=0),Scale3D=(X=1,Y=1,Z=1))))
    Mirror Modify Data=((Item=(Type=Bone,Name="deform_elbow_in_r"),Transform=(Rotation=(X=0,Y=0,Z=0,W=1),Translation=(X=0,Y=0,Z=0),Scale3D=(X=1,Y=1,Z=1))))
    Weight=&RigVMFunction_MathFloatRemap(SourceMaximum=90, SourceMinimum=15, bClamp=true, Value=@"Get Delta Ref".Result.Rotation.Yaw).Result
    "Mirror Weight "=&RigVMFunction_MathFloatRemap(SourceMaximum=90, SourceMinimum=15, bClamp=true, Value=@"Get Delta Ref_1".Result.Rotation.Yaw).Result
   )
   > RigVMFunctionReturnNode()
  ]}
 New Function_ContainedGraph {nodes [
   RigVMFunctionEntryNode()
   > &RigUnit_ParentConstraint(
    AdvancedSettings=(InterpolationType=Average,RotationOrderForFilter=XZY)
    Child=(Type=Bone,Name="deform_knee_in_l_base")
    Parents=((Item=(Type=Bone,Name="deform_thigh_l"),Weight=1),(Item=(Type=Bone,Name="calf_l"),Weight=1))
    Weight=1
   )
   > &RigUnit_ParentConstraint(
    AdvancedSettings=(InterpolationType=Average,RotationOrderForFilter=XZY)
    Child=(Type=Bone,Name="deform_knee_out_l_base")
    Parents=((Item=(Type=Bone,Name="deform_thigh_l"),Weight=1),(Item=(Type=Bone,Name="calf_l"),Weight=1))
    Weight=1
   )
   > &RigUnit_ParentConstraint(
    AdvancedSettings=(InterpolationType=Average,RotationOrderForFilter=XZY)
    Child=(Type=Bone,Name="deform_knee_out_r_base")
    Parents=((Item=(Type=Bone,Name="deform_thigh_r"),Weight=1),(Item=(Type=Bone,Name="calf_r"),Weight=1))
    Weight=1
   )
   > &RigUnit_ParentConstraint(
    AdvancedSettings=(InterpolationType=Average,RotationOrderForFilter=XZY)
    Child=(Type=Bone,Name="deform_knee_in_r_base")
    Parents=((Item=(Type=Bone,Name="deform_thigh_r"),Weight=1),(Item=(Type=Bone,Name="calf_r"),Weight=1))
    Weight=1
   )
   > Get Delta Ref_2 = RigVMFunctionReferenceNode(Item=(Type=Bone,Name="calf_l"))
   > Get Delta Ref_3 = RigVMFunctionReferenceNode(Item=(Type=Bone,Name="calf_r"))
   > Modify Transforms with Mirror_1 = RigVMFunctionReferenceNode(
    Base Modify Data="((Item=(Type=Bone,Name=\"deform_knee_in_l\"),Transform=(Rotation=(X=0,Y=0,Z=0,W=1),Translation=(X=-3,Y=-7,Z=0),Scale3D=(X=1,Y=1,Z=1))),(Item=(Type=Bone,Name=\"deform_knee_out_l\"),Transform=(Rotation=(X=0,Y=0,Z=0,W=1),Translation=(X=3,Y=5,Z=0),Scale3D=(X=0.5,Y ...(truncated)"
    Mirror Modify Data="((Item=(Type=Bone,Name=\"deform_knee_in_r\"),Transform=(Rotation=(X=0,Y=0,Z=0,W=1),Translation=(X=0,Y=0,Z=0),Scale3D=(X=1,Y=1,Z=1))),(Item=(Type=Bone,Name=\"deform_knee_out_r\"),Transform=(Rotation=(X=0,Y=0,Z=0,W=1),Translation=(X=0,Y=0,Z=0),Scale3D=(X=1,Y=1,Z ...(truncated)"
    Weight=&RigVMFunction_MathFloatRemap(SourceMaximum=100, SourceMinimum=10, bClamp=true, Value=@"Get Delta Ref_2".Result.Rotation.Yaw).Result
    "Mirror Weight "=&RigVMFunction_MathFloatRemap(SourceMaximum=100, SourceMinimum=10, bClamp=true, Value=@"Get Delta Ref_3".Result.Rotation.Yaw).Result
   )
   > RigVMFunctionReturnNode()
  ]}
 New Function_ContainedGraph {nodes [
   RigVMFunctionEntryNode()
   Deform Fingers = RigVMDispatchNode(Value="((Type=Bone,Name=\"deform_thumb_01_l\"),(Type=Bone,Name=\"deform_thumb_02_l\"),(Type=Bone,Name=\"deform_thumb_03_l\"),(Type=Bone,Name=\"deform_index_metacarpal_l\"),(Type=Bone,Name=\"deform_index_01_l\"),(Type=Bone,Name=\"deform_index_02_l\"),(Type=Bone,Name=\"deform_i ...(truncated)") {template="DISPATCH_RigVMDispatch_Constant(io Value)"}
   > For_Each_7 = RigVMDispatchNode(Array=@"Deform Fingers".Value) {template="DISPATCH_RigVMDispatch_ArrayIterator(in Array,out Element,out Index,out Count,out Ratio)"}
   Base Fingers = RigVMDispatchNode(Value="((Type=Bone,Name=\"thumb_01_l\"),(Type=Bone,Name=\"thumb_02_l\"),(Type=Bone,Name=\"thumb_03_l\"),(Type=Bone,Name=\"index_metacarpal_l\"),(Type=Bone,Name=\"index_01_l\"),(Type=Bone,Name=\"index_02_l\"),(Type=Bone,Name=\"index_03_l\"),(Type=Bone,Name=\"middle_metacarpal_l\" ...(truncated)") {template="DISPATCH_RigVMDispatch_Constant(io Value)"}
   > &RigUnit_ParentConstraint(
    AdvancedSettings=(InterpolationType=Average,RotationOrderForFilter=XZY)
    Parents=((Item=(Type=Bone,Name="None"),Weight=1))
    Weight=1
    Parents.0.Item=RigVMDispatchNode(Array=@"Base Fingers".Value, Index=@For_Each_7.Index) {template="DISPATCH_RigVMDispatch_ArrayGetAtIndex(in Array,in Index,out Element)"}.Element
    Child=@For_Each_7.Element
   )
   RigVMFunctionReturnNode(ExecuteContext=@For_Each_7.Completed)
  ]}
 New Function_ContainedGraph {nodes [
   RigVMFunctionEntryNode()
   Multiply = &RigVMFunction_MathTransformMul(
    B=/Script/RigVM.RigVMFunction_MathTransformInverse(Value=&RigUnit_GetTransform(Item=(Type=Bone,Name="spine_05"), Space=LocalSpace, bInitial=true).Transform) {is=struct RigVMUnitNode, template="Inverse::Execute(in Value,out Result)"}.Result
    A=&RigUnit_GetTransform(Item=(Type=Bone,Name="spine_05"), Space=LocalSpace, bInitial=False).Transform
   ) {template="Multiply::Execute(in A,in B,out Result)"}
   n1 = &RigVMFunction_MathQuaternionSwingTwist(TwistAxis=(X=1,Y=0,Z=0), Input=@Multiply.Result.Rotation)
   > RigUnit_SetRotation = &RigUnit_SetRotation(Item=(Type=Bone,Name="deform_spine_04"), Weight=1, bInitial=False, Value=@n1.Twist)
   > &RigUnit_SetRotation(Item=(Type=Bone,Name="deform_spine_03"), Weight=1, bInitial=False, Value=@n1.Twist)
   > &RigUnit_SetRotation(Item=(Type=Bone,Name="deform_spine_02"), Weight=1, bInitial=False, Value=@n1.Twist)
   > &RigUnit_SetRotation(Item=(Type=Bone,Name="deform_spine_01"), Weight=1, bInitial=False, Value=@n1.Twist)
   > &RigUnit_SetRotation(Item=(Type=Bone,Name="deform_pelvis"), Weight=1, bInitial=False, Value=@n1.Twist)
   > RigVMFunctionReturnNode()
  ]}
 New Function_ContainedGraph {nodes [
   n1 = RigVMFunctionEntryNode()
   n2 = &RigUnit_GetTransform(bInitial=False, Item=@n1.Start)
   Subtract = /Script/RigVM.RigVMFunction_MathVectorSub(B=@n2.Transform.Translation, A=&RigUnit_GetTransform(bInitial=False, Item=@n1.End).Transform.Translation) {is=struct RigVMUnitNode, template="Subtract::Execute(in A,in B,out Result)"}
   Unit = /Script/RigVM.RigVMFunction_MathVectorUnit(Value=@Subtract.Result) {is=struct RigVMUnitNode, template="Unit::Execute(in Value,out Result)"}
   RigVMFunction_MathQuaternionFromAxisAndAngle = /Script/RigVM.RigVMFunction_MathQuaternionFromAxisAndAngle(Axis=@Unit.Result, Angle=@n1.QuatAmount) {is=struct RigVMUnitNode, template="FromAxisAndAngle::Execute(in Axis,in Angle,out Result)"}
   Multiply = /Script/RigVM.RigVMFunction_MathQuaternionMul(B=@n2.Transform.Rotation, A=@RigVMFunction_MathQuaternionFromAxisAndAngle.Result) {is=struct RigVMUnitNode, template="Multiply::Execute(in A,in B,out Result)"}
   > &RigUnit_SetRotation(Space=GlobalSpace, bInitial=False, Item=@n1.Start, Weight=@n1.Weight, Value=@Multiply.Result)
   > RigVMFunctionReturnNode()
  ]}
 RigVMFunctionLibrary {nodes [
   ApplyFBIKChainCorrection = RigVMCollapseNode(End=(Type=None,Name="None"), ExecuteContext=(), Start=(Type=None,Name="None"))
   Deform Elbows = RigVMCollapseNode(ExecuteContext=())
   Deform Knees = RigVMCollapseNode(ExecuteContext=())
   Deform Spine = RigVMCollapseNode(ExecuteContext=())
   DeformRig_GetDeltaRefTransform_FNC = RigVMCollapseNode(ExecuteContext=(), Item=(Type=Bone))
   Retarget Arms = RigVMCollapseNode(ExecuteContext=())
   Retarget Fingers = RigVMCollapseNode(ExecuteContext=())
   Retarget Legs = RigVMCollapseNode(ExecuteContext=())
  ]}
 RigVMFunction_Sequence_1_ContainedGraph {nodes [
   n1 = RigVMFunctionEntryNode()
   RigVMFunction_Sequence = &RigVMFunction_Sequence(ExecuteContext=@n1.ExecuteContext)
   n2 = &RigVMFunction_Sequence(ExecuteContext=@RigVMFunction_Sequence.B)
   n3 = &RigVMFunction_Sequence(ExecuteContext=@n2.B)
   RigVMFunctionReturnNode(A=@RigVMFunction_Sequence.A, B=@n2.A, C=@n3.A, D=@n3.B)
  ]}
 RigVMModel {nodes [
   RigUnit_BeginExecution = &RigUnit_BeginExecution()
   Retarget Arms = RigVMFunctionReferenceNode(ExecuteContext=@RigUnit_BeginExecution.ExecutePin)
   > Retarget Legs = RigVMFunctionReferenceNode()
   > Retarget Fingers = RigVMFunctionReferenceNode()
   > Deform Elbows = RigVMFunctionReferenceNode()
   > Deform Knees = RigVMFunctionReferenceNode()
   > Deform Spine = RigVMFunctionReferenceNode()
  ]}
]
hierarchy [
 root {loc=<0 0 0>, rot=<0 0 0 1>}
 pelvis {parent=root, loc=<0 -0.7321 92.4218>, rot=<0.0127 0.707 -0.0127 -0.707>}
 spine_01 {parent=pelvis, loc=<5.6002 -0.5128 0>, rot=<0 0 0.034 -0.9994>}
 spine_02 {parent=spine_01, loc=<8.0756 0.0139 0>, rot=<0 0 0.0238 -0.9997>}
 spine_03 {parent=spine_02, loc=<7.0395 0 0>, rot=<0 0 0.0149 -0.9999>}
 spine_04 {parent=spine_03, loc=<8.1023 0 0>, rot=<0 0 -0.0132 -0.9999>}
 spine_05 {parent=spine_04, loc=<10.0141 0 0>, rot=<0 0 -0.0015 -1>}
 clavicle_l {parent=spine_05, loc=<2.5856 -0.2866 -4.3777>, rot=<0.0465 -0.7289 -0.009 -0.6829>}
 upperarm_l {parent=clavicle_l, loc=<18.2264 0.1342 -0.001>, rot=<0.0253 -0.2693 0.0013 -0.9627>}
 lowerarm_l {parent=upperarm_l, loc=<29.0071 0.1686 -1.239>, rot=<0.0391 -0.0099 -0.2453 -0.9686>}
 hand_l {parent=lowerarm_l, loc=<25.8607 0.222 1.0274>, rot=<-0.7193 -0.0988 0.1193 -0.6773>}
 thumb_01_l {parent=hand_l, loc=<1.3118 -0.3191 -2.459>, rot=<0.6546 -0.4042 0.0664 -0.6354>}
 thumb_02_l {parent=thumb_01_l, loc=<5.9336 -0.0448 0.0401>, rot=<-0.0037 0.0213 0.298 -0.9543>}
 thumb_03_l {parent=thumb_02_l, loc=<2.2223 0.3682 -0.0907>, rot=<0.0118 -0.013 0.0066 -0.9998>}
 index_metacarpal_l {parent=hand_l, loc=<1.6535 0.0183 -2.96>, rot=<0.0158 -0.0124 -0.0173 -0.9996>}
 index_01_l {parent=index_metacarpal_l, loc=<8.3008 0.0357 -0.0524>, rot=<-0.0044 0.0485 0.0924 -0.9945>}
 index_02_l {parent=index_01_l, loc=<3.3587 -0.3004 0.0452>, rot=<-0.002 0.0027 0.3225 -0.9466>}
 index_03_l {parent=index_02_l, loc=<2.5578 -0.123 -0.0252>, rot=<-0.002 0.0067 0.2558 -0.9667>}
 middle_metacarpal_l {parent=hand_l, loc=<2.1856 -0.1725 -0.5984>, rot=<0.006 0.0168 -0.0329 -0.9993>}
 middle_01_l {parent=middle_metacarpal_l, loc=<7.3275 -0.2447 0.0129>, rot=<-0.0208 0.0042 0.0858 -0.9961>}
 middle_02_l {parent=middle_01_l, loc=<4.0123 0.1401 0.1175>, rot=<-0.0218 0.0258 0.3512 -0.9357>}
 middle_03_l {parent=middle_02_l, loc=<2.6077 0.0534 0.1373>, rot=<0.0082 0.0188 0.1708 -0.9851>}
 ring_metacarpal_l {parent=hand_l, loc=<2.2074 -0.3625 1.3638>, rot=<-0.0284 0.0722 -0.0458 -0.9959>}
 ring_01_l {parent=ring_metacarpal_l, loc=<7.0295 -0.6197 -0.1446>, rot=<-0.0489 -0.0252 0.1892 -0.9804>}
 ring_02_l {parent=ring_01_l, loc=<3.6174 0.4647 -0.1027>, rot=<0 0 0.2806 -0.9598>}
 ring_03_l {parent=ring_02_l, loc=<2.4152 0.114 0.0199>, rot=<0 0 0.1189 -0.9929>}
 pinky_metacarpal_l {parent=hand_l, loc=<1.7616 -0.9753 3.3814>, rot=<-0.0397 0.0814 -0.069 -0.9935>}
 pinky_01_l {parent=pinky_metacarpal_l, loc=<7.1709 -0.4831 -0.1094>, rot=<-0.0654 -0.0035 0.2242 -0.9723>}
 pinky_02_l {parent=pinky_01_l, loc=<2.8274 0.0893 -0.1043>, rot=<0.0037 0.0011 0.2854 -0.9584>}
 pinky_03_l {parent=pinky_02_l, loc=<2.2046 0.0307 -0.0078>, rot=<-0.0002 0.0098 0.1445 -0.9895>}
 weapon_l {parent=hand_l, loc=<-0.1974 -3.4111 0.3811>, rot=<0 0 0 1>}
 hand_attach_l {parent=hand_l, loc=<8.5 -3 0>, rot=<0.7071 0.7071 0 0>}
 cr_lowerarm_push_l {parent=lowerarm_l, loc=<0 0 0>, rot=<0 0 0 1>}
 neck_01 {parent=spine_05, loc=<10.815 1.9196 0>, rot=<0 0 -0.1864 -0.9825>}
 neck_02 {parent=neck_01, loc=<4.7347 0.3045 0>, rot=<0 0 0.0363 -0.9993>}
 head {parent=neck_02, loc=<5.7632 0.5552 0>, rot=<0 0 0.115 -0.9934>}
 clavicle_r {parent=spine_05, loc=<2.5855 -0.2866 4.3777>, rot=<0.7289 0.0465 -0.6829 0.009>}
 upperarm_r {parent=clavicle_r, loc=<-18.2264 -0.1342 0.001>, rot=<0.0253 -0.2693 0.0013 -0.9627>}
 lowerarm_r {parent=upperarm_r, loc=<-29.0071 -0.1686 1.239>, rot=<0.0391 -0.0099 -0.2453 -0.9686>}
 hand_r {parent=lowerarm_r, loc=<-25.8607 -0.222 -1.0274>, rot=<-0.7193 -0.0988 0.1193 -0.6773>}
 thumb_01_r {parent=hand_r, loc=<-1.3118 0.3191 2.459>, rot=<0.6546 -0.4042 0.0664 -0.6354>}
 thumb_02_r {parent=thumb_01_r, loc=<-5.9336 0.0448 -0.0401>, rot=<-0.0037 0.0213 0.298 -0.9543>}
 thumb_03_r {parent=thumb_02_r, loc=<-2.2223 -0.3682 0.0907>, rot=<0.0118 -0.013 0.0066 -0.9998>}
 index_metacarpal_r {parent=hand_r, loc=<-1.6535 -0.0183 2.96>, rot=<0.0158 -0.0124 -0.0173 -0.9996>}
 index_01_r {parent=index_metacarpal_r, loc=<-8.3008 -0.0357 0.0524>, rot=<-0.0044 0.0485 0.0924 -0.9945>}
 index_02_r {parent=index_01_r, loc=<-3.3587 0.3004 -0.0452>, rot=<-0.002 0.0027 0.3225 -0.9466>}
 index_03_r {parent=index_02_r, loc=<-2.5578 0.123 0.0252>, rot=<-0.002 0.0067 0.2558 -0.9667>}
 middle_metacarpal_r {parent=hand_r, loc=<-2.1856 0.1725 0.5984>, rot=<0.006 0.0168 -0.0329 -0.9993>}
 middle_01_r {parent=middle_metacarpal_r, loc=<-7.3275 0.2447 -0.0129>, rot=<-0.0208 0.0042 0.0858 -0.9961>}
 middle_02_r {parent=middle_01_r, loc=<-4.0123 -0.1401 -0.1175>, rot=<-0.0218 0.0258 0.3512 -0.9357>}
 middle_03_r {parent=middle_02_r, loc=<-2.6077 -0.0534 -0.1373>, rot=<0.0082 0.0188 0.1708 -0.9851>}
 ring_metacarpal_r {parent=hand_r, loc=<-2.2074 0.3625 -1.3638>, rot=<-0.0284 0.0722 -0.0458 -0.9959>}
 ring_01_r {parent=ring_metacarpal_r, loc=<-7.0295 0.6197 0.1446>, rot=<-0.0489 -0.0252 0.1892 -0.9804>}
 ring_02_r {parent=ring_01_r, loc=<-3.6174 -0.4647 0.1027>, rot=<0 0 0.2806 -0.9598>}
 ring_03_r {parent=ring_02_r, loc=<-2.4152 -0.114 -0.0199>, rot=<0 0 0.1189 -0.9929>}
 pinky_metacarpal_r {parent=hand_r, loc=<-1.7616 0.9753 -3.3814>, rot=<-0.0397 0.0814 -0.069 -0.9935>}
 pinky_01_r {parent=pinky_metacarpal_r, loc=<-7.1709 0.4831 0.1094>, rot=<-0.0654 -0.0035 0.2242 -0.9723>}
 pinky_02_r {parent=pinky_01_r, loc=<-2.8274 -0.0893 0.1043>, rot=<0.0037 0.0011 0.2854 -0.9584>}
 pinky_03_r {parent=pinky_02_r, loc=<-2.2046 -0.0307 0.0078>, rot=<-0.0002 0.0098 0.1445 -0.9895>}
 hand_attach_r {parent=hand_r, loc=<-8.5 3 0>, rot=<0 0 -0.7071 -0.7071>}
 lowerarm_twist_02_r {parent=lowerarm_r, loc=<-9.6199 0 0>, rot=<0 0 0 -1>}
 cr_lowerarm_push_r {parent=lowerarm_r, loc=<0 0 0>, rot=<0 0 0 1>}
 upperarm_twist_01_r {parent=upperarm_r, loc=<-7.9796 -0.0843 0.6195>, rot=<0 0 0 -1>}
 upperarm_twist_02_r {parent=upperarm_r, loc=<-19.5059 -0.0843 0.6195>, rot=<0 0 0 -1>}
 attach_cape {parent=spine_05, loc=<9.017 -11.714 0>, rot=<0.0616 -0.7044 -0.0616 -0.7044>}
 cape {parent=attach_cape, type=Null, loc=<0 -4.6715 -0.7773>, rot=<0.0539 0 0 0.9985>}
 attach_backpack {parent=spine_05, loc=<-0.123 -13.473 0>, rot=<-0.0093 -0.707 0.0093 -0.707>}
 Backpack_BR {parent=attach_backpack, type=Null, loc=<0 -5.2725 -1.8311>, rot=<-0.0674 0 0 0.9977>}
 deform_clavicle_l {parent=spine_05, loc=<2.206 2.0367 0>, rot=<0.0284 -0.7065 -0.0284 -0.7065>}
 deform_upperarm_l {parent=deform_clavicle_l, loc=<22.5467 -3.1283 -1.0316>, rot=<0.0177 -0.302 0.0245 -0.9528>}
 deform_lowerarm_l {parent=deform_upperarm_l, loc=<29.0071 0.1686 -1.239>, rot=<0.0391 -0.0099 -0.2453 -0.9686>}
 deform_hand_l {parent=deform_lowerarm_l, loc=<25.8607 0.222 1.0274>, rot=<-0.7193 -0.0988 0.1193 -0.6773>}
 deform_thumb_01_l {parent=deform_hand_l, loc=<1.479 -1.6102 -2.6917>, rot=<0.653 -0.4169 -0.0324 -0.6315>}
 deform_thumb_02_l {parent=deform_thumb_01_l, loc=<2.7711 0 0>, rot=<0.0229 0.0018 0.1715 -0.9849>}
 deform_thumb_03_l {parent=deform_thumb_02_l, loc=<2.8017 0 0>, rot=<0 0 0.1573 -0.9875>}
 deform_index_metacarpal_l {parent=deform_hand_l, loc=<3.3047 0.9878 -1.6966>, rot=<0.158 -0.0721 -0.0551 -0.9833>}
 deform_index_01_l {parent=deform_index_metacarpal_l, loc=<3.5019 0 0>, rot=<0.0054 -0.0179 0.1445 -0.9893>}
 deform_index_02_l {parent=deform_index_01_l, loc=<4.1431 0 0>, rot=<-0.0476 -0.0151 0.1728 -0.9837>}
 deform_index_03_l {parent=deform_index_02_l, loc=<4 0 0>, rot=<-0.0544 -0.0056 0.1027 -0.9932>}
 deform_middle_metacarpal_l {parent=deform_hand_l, loc=<3.6656 1.0013 0.6129>, rot=<0.0352 -0.0065 -0.0523 -0.998>}
 deform_middle_01_l {parent=deform_middle_metacarpal_l, loc=<3.2785 0 0>, rot=<-0.0136 -0.0137 0.1397 -0.99>}
 deform_middle_02_l {parent=deform_middle_01_l, loc=<3.8221 0 0>, rot=<-0.004 -0.0063 0.1574 -0.9875>}
 deform_middle_03_l {parent=deform_middle_02_l, loc=<4 0 0>, rot=<-0.0492 0.0131 0.0825 -0.9953>}
 deform_elbow_in_l_base {parent=deform_lowerarm_l, loc=<0 2.3826 0>, rot=<0 0 0 -1>}
 deform_elbow_in_l {parent=deform_elbow_in_l_base, loc=<0 0 0>, rot=<0 0 0 1>}
 deform_upperarm_l_end {parent=deform_upperarm_l, loc=<25.4509 0.1686 -1.239>, rot=<0 0 0 -1>}
 deform_clavicle_r {parent=spine_05, loc=<2.2059 2.0367 0>, rot=<0.7065 0.0284 -0.7065 0.0284>}
 deform_upperarm_r {parent=deform_clavicle_r, loc=<-22.5467 3.1283 1.032>, rot=<0.0177 -0.302 0.0245 -0.9528>}
 deform_lowerarm_r {parent=deform_upperarm_r, loc=<-29.0067 -0.1685 1.2385>, rot=<0.0391 -0.0099 -0.2453 -0.9686>}
 deform_hand_r {parent=deform_lowerarm_r, loc=<-25.8609 -0.2219 -1.0273>, rot=<-0.7193 -0.0988 0.1193 -0.6773>}
 deform_thumb_01_r {parent=deform_hand_r, loc=<-1.4788 1.6101 2.6918>, rot=<0.653 -0.4169 -0.0324 -0.6315>}
 deform_thumb_02_r {parent=deform_thumb_01_r, loc=<-2.7712 0 0>, rot=<0.0229 0.0018 0.1715 -0.9849>}
 deform_thumb_03_r {parent=deform_thumb_02_r, loc=<-2.8016 -0.0001 0>, rot=<0 0 0.1573 -0.9875>}
 deform_index_metacarpal_r {parent=deform_hand_r, loc=<-3.3045 -0.9878 1.6966>, rot=<0.158 -0.0721 -0.0551 -0.9833>}
 deform_index_01_r {parent=deform_index_metacarpal_r, loc=<-3.5019 0 0>, rot=<0.0054 -0.0179 0.1445 -0.9893>}
 deform_index_02_r {parent=deform_index_01_r, loc=<-4.143 0 0>, rot=<-0.0476 -0.0151 0.1728 -0.9837>}
 deform_index_03_r {parent=deform_index_02_r, loc=<-4 0 -0.0001>, rot=<-0.0544 -0.0056 0.1027 -0.9932>}
 deform_middle_metacarpal_r {parent=deform_hand_r, loc=<-3.6655 -1.0014 -0.6129>, rot=<0.0352 -0.0065 -0.0523 -0.998>}
 deform_middle_01_r {parent=deform_middle_metacarpal_r, loc=<-3.2785 0.0001 0>, rot=<-0.0136 -0.0137 0.1397 -0.99>}
 deform_middle_02_r {parent=deform_middle_01_r, loc=<-3.8221 0 0>, rot=<-0.004 -0.0063 0.1574 -0.9875>}
 deform_middle_03_r {parent=deform_middle_02_r, loc=<-4 0 0>, rot=<-0.0492 0.0131 0.0825 -0.9953>}
 deform_elbow_in_r_base {parent=deform_lowerarm_r, loc=<-0.0003 -2.3825 0.0004>, rot=<0 0 0 -1>}
 deform_elbow_in_r {parent=deform_elbow_in_r_base, loc=<0 0 0>, rot=<0 0 0 -1>}
 deform_upperarm_r_end {parent=deform_upperarm_r, loc=<-25.4509 -0.1686 1.2389>, rot=<0 0 0 -1>}
 deform_spine_05 {parent=spine_05, loc=<0 0 0>, rot=<0 0 0 1>}
 deform_spine_04 {parent=spine_04, loc=<0 0 0>, rot=<0 0 0 1>}
 deform_spine_03 {parent=spine_03, loc=<0 0 0>, rot=<0 0 0 1>}
 deform_spine_02 {parent=spine_02, loc=<0 0 0>, rot=<0 0 0 1>}
 deform_spine_01 {parent=spine_01, loc=<0 0 0>, rot=<0 0 0 1>}
 thigh_l {parent=pelvis, loc=<-1.8716 -0.0109 -11.9936>, rot=<0.0012 0.0572 0.0155 -0.9982>}
 calf_l {parent=thigh_l, loc=<-41.9178 -0.2246 0.0926>, rot=<0.0006 -0.0161 -0.0391 -0.9991>}
 foot_l {parent=calf_l, loc=<-38.7057 1.2438 0.2452>, rot=<0.0975 -0.0474 0.0317 -0.9936>}
 ball_l {parent=foot_l, loc=<-6.1614 13.1469 0.6807>, rot=<0.0001 -0.0001 -0.7063 -0.7079>}
 NewBone {parent=ball_l, loc=<-14.098 0 0>, rot=<0 0 0 1>}
 calf_twist_02_l {parent=calf_l, loc=<-15.9268 -0.0967 0.0397>, rot=<0.0237 0.0095 -0.0035 -0.9997>}
 calf_twist_01_l {parent=calf_l, loc=<-32.9335 -0.0967 0.0397>, rot=<0.0237 0.0095 -0.0035 -0.9997>}
 thigh_twist_01_l {parent=thigh_l, loc=<-21.6303 0 0>, rot=<0.0013 0 0 -1>}
 thigh_r {parent=pelvis, loc=<-1.8716 -0.0109 11.9936>, rot=<0.0572 -0.0012 0.9982 0.0155>}
 calf_r {parent=thigh_r, loc=<41.9178 0.2246 -0.0926>, rot=<0.0006 -0.0161 -0.0391 -0.9991>}
 foot_r {parent=calf_r, loc=<38.7057 -1.2438 -0.2452>, rot=<0.0975 -0.0474 0.0317 -0.9936>}
 ball_r {parent=foot_r, loc=<6.1614 -13.1469 -0.6807>, rot=<0.0001 -0.0001 -0.7063 -0.7079>}
 calf_twist_02_r {parent=calf_r, loc=<15.9268 0.0967 -0.0397>, rot=<0.0237 0.0095 -0.0035 -0.9997>}
 calf_twist_01_r {parent=calf_r, loc=<32.9335 0.0967 -0.0397>, rot=<0.0237 0.0095 -0.0035 -0.9997>}
 NewBone_2 {parent=calf_r, loc=<44.1008 -14.6229 1.6272>, rot=<0.0357 -0.1025 -0.6793 -0.7258>}
 thigh_twist_01_r {parent=thigh_r, loc=<21.6303 0 0>, rot=<0.0013 0 0 -1>}
 deform_thigh_l {parent=pelvis, loc=<-7.2263 -0.1565 -11.7579>, rot=<0.046 0.0666 0.0196 -0.9965>}
 deform_calf_l {parent=deform_thigh_l, loc=<-36.6398 0 0>, rot=<-0.0444 -0.0245 -0.0414 -0.9979>}
 deform_knee_in_l_base {parent=deform_calf_l, loc=<0 -3.8517 0>, rot=<0 0 0 1>}
 deform_knee_in_l {parent=deform_knee_in_l_base, loc=<0 0 0>, rot=<0 0 0 1>}
 deform_knee_out_l_base {parent=deform_calf_l, loc=<0 3.4446 0>, rot=<0 0 0 1>}
 deform_knee_out_l {parent=deform_knee_out_l_base, loc=<0 0 0>, rot=<0 0 0 1>}
 deform_thigh_r {parent=pelvis, loc=<-7.2263 -0.1565 11.7579>, rot=<0.0666 -0.046 0.9965 0.0196>}
 deform_calf_r {parent=deform_thigh_r, loc=<36.6398 0 0>, rot=<-0.0444 -0.0245 -0.0414 -0.9979>}
 deform_knee_out_r_base {parent=deform_calf_r, loc=<0 -3.4446 0>, rot=<0 0 0 1>}
 deform_knee_out_r {parent=deform_knee_out_r_base, loc=<0 0 0>, rot=<0 0 0 -1>}
 deform_knee_in_r_base {parent=deform_calf_r, loc=<0 3.8517 0>, rot=<0 0 0 1>}
 deform_knee_in_r {parent=deform_knee_in_r_base, loc=<0 0 0>, rot=<0 0 0 -1>}
 deform_pelvis {parent=pelvis, loc=<0 0 0>, rot=<0 0 0 1>}
]
variables [Wearing Shoes {type=bool, default=False}]
