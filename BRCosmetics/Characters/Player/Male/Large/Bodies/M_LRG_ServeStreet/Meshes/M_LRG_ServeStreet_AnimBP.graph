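# Animation Blueprint for the M_LRG_ServeStreet large male player body cosmetic, adjusting its pose based on whether the character wears shoes.
# Copies the pose from the attached parent mesh and runs it through the body's Control Rig, feeding in a Wearing Shoes flag.
# When character customization loads, it looks up the ExtraPart mesh as the shoe and sets that flag from whether one exists.
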AnimBlueprint /BRCosmetics/Characters/Player/Male/Large/Bodies/M_LRG_ServeStreet/Meshes/M_LRG_ServeStreet_AnimBP
target_skeleton=&/M_LRG_ServeStreet_Skeleton
graphs [
 AnimGraph {nodes [
   n1 = Control Rig!(Source=Copy Pose From Mesh() {bUseAttachedParent=true}, Alpha, Wearing Shoes:bool=Get Wearing Shoes() -> (Wearing Shoes:bool)) -> (Pose) {
    alphaBoolBlend {blendOption=Linear}
    controlRigAssetReference {blueprintRigClass=&/M_LRG_ServeStreet_CtrlRig.M_LRG_ServeStreet_CtrlRig_C}
   }
   Output Pose(Result=@n1) {layerGroup=None}
  ]}
 EventGraph {nodes [
   n1 = Event OnCharacterCustomizationLoaded()
   > n2 = Set ShoeMesh(ShoeMesh:Skeletal Mesh Component Budgeted Object Reference=Get Array Item with EFort Custom Part Type(Array=@n1.CustomMeshParts, PartType=ExtraPart)) -> (Output_Get:Skeletal Mesh Component Budgeted Object Reference)
   > n3 = Is Valid(InputObject=@n2)
   Set Wearing Shoes(execute=@n3.Is Not Valid, Wearing Shoes:bool=false) -> (Output_Get:bool)
   Set Wearing Shoes(execute=@n3.Is Valid, Wearing Shoes=true)
   Event BlueprintUpdateAnimation()
  ]}
]
variables [
 ShoeMesh {type=SkeletalMeshComponentBudgeted}
 Wearing Shoes {type=boolean}
]
