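# Physics asset of cape colliders for the large male ServeStreet character body, shaped for cape collision rather than ragdoll.
# Capsules cover the pelvis, three spine bones, neck and both mirrored clavicles, with a single sphere on the head.
# Upper arm, forearm, thigh, calf and foot bodies exist but hold no shapes, so only the torso and head collide.
# No constraints join the bodies.
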PhysicsAsset /BRCosmetics/Characters/Player/Male/Large/Bodies/M_LRG_ServeStreet/Meshes/M_LRG_ServeStreet_CapeColliders
bodies {
 calf_l []
 calf_r []
 clavicle_l [{shapeType=Capsule, length=9.0192, radius=9.8871, center=<9.4266 -1.1806 -2.4977>, rotation=<rot 69.3297 1.4476 176.9462>}]
 clavicle_r [{shapeType=Capsule, length=9.0192, radius=9.8871, center=<-9.4266 1.1806 2.4977>, rotation=<rot 69.3297 1.4476 -3.0538>}]
 foot_l []
 foot_r []
 head [{shapeType=Sphere, radius=17.3125, center=<11.028 -0.174 0>}]
 lowerarm_l []
 lowerarm_r []
 neck_02 [{shapeType=Capsule, length=13.4375, radius=7.3175, center=<-0.2642 -0.3408 0>, rotation=<rot 90 -171.1193 -175.3978>}]
 pelvis [{shapeType=Capsule, length=0.0403, radius=22.7415, center=<1.3383 0.1881 0>, rotation=<rot 90 0 0>}]
 spine_01 [{shapeType=Capsule, length=7.9438, radius=13.5586, center=<-1.7965 0.1884 0>}]
 spine_02 [{shapeType=Capsule, length=19.9518, radius=21.0244, center=<0.0474 -0.4465 0>, rotation=<rot 90 6.0557 0>}]
 spine_04 [{shapeType=Capsule, length=17.3141, radius=20.9514, shapeName=spine_04_capsule, center=<4.2155 1.6102 0>, rotation=<rot 90 4.9722 0>}]
 thigh_l []
 thigh_r []
 upperarm_l []
 upperarm_r []
}
constraints []
